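# Additive, unlit, two-sided Niagara mesh-particle material for the WealthLamb impact shockwave, with tinted glow and refraction distortion.
# Named reroutes define a per-vertex Fresnel, a depth mask bright near scene intersections, and a camera-distance fade.
# Alpha samples a soft ramp at the maximum of two routed masks, dimming back faces.
# Particle alpha modulates the tint-alpha blend; refraction follows a view-transformed vector, scaled by dynamic parameter Ref and optionally pixel depth.
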Material /PistolCanary/Assets/FX/WealthLamb/ImpactShockwave/M_PC_WealthLamb_Shockwave
graph {
 properties {
  bAllowNegativeEmissiveColor=true
  bUsedWithNiagaraMeshParticles=true
  blendMode=BLEND_Additive
  pixelDepthOffsetMode=PDOM_Legacy
  refractionMethod=RM_2DOffset
  shadingModel=MSM_Unlit
  twoSided=true
  usedShadingModels=Unlit
 }
 nodes [
  n1 = VectorParameter() {group=Emissive, parameterName=EmissiveColorTintA}
  n2 = VectorParameter() {group=Emissive, parameterName=EmissiveColorTintB, defaultValue=<rgba 0.0066 0.0145 0.0174 1>}
  LinearInterpolate(A=@n1.RGB, B=@n2.RGB, Alpha=NamedRerouteUsage())
  LinearInterpolate(A=@n1.A, B=@n2.A, Alpha=Multiply(A=ParticleColor().A, B=NamedRerouteUsage()))
  n3 = Transform(None=VectorParameter() {group=Refraction, parameterName=RefractionVector, defaultValue=<rgba 0 0 -1 0>}.RGB) {transformType=TRANSFORM_View}
  Multiply(A=@n3, B=ScalarParameter() {defaultValue=-0.1387, group=Refraction, parameterName=Refraction})
  n4 = Multiply(A=Multiply(A=@n3, B=NamedRerouteUsage()), B=DynamicParameter() {paramNames [Param1,Param2,Param3,Ref]}.Ref)
  n5 = LinearInterpolate(B=Divide(B=PixelDepth().R) {constA=1000}, Alpha=ScalarParameter() {defaultValue=1, parameterName=DepthAffectRefraction}) {constA=1}
  Multiply(A=ComponentMask(None=Multiply(A=@n4, B=@n5)) {g=true, r=true}, B=NamedRerouteUsage())
  n6 = Max(A=NamedRerouteUsage(), B=NamedRerouteUsage())
  n7 = ScalarParameter() {defaultValue=0.5, parameterName=BackSideStrength}
  n8 = Saturate(None=TwoSidedSign())
  NamedRerouteDeclaration(None=Multiply(
    A=TextureSample(UVs=@n6) {texture=/Game/Effects/Fort_Effects/Textures/Ramps/T_Power_Source_Sphere_Ramp_Soft}.G
    B=LinearInterpolate(A=@n7, Alpha=@n8)
   )) {name=Alpha, nodeColor=<rgba 1 1 1 1>}
  n9 = Subtract(A=CameraPositionWS(), B=WorldPosition().XYZ)
  n10 = Length(None=@n9) {is=class MaterialExpressionLength}
  n11 = DotProduct(A=Normalize(VectorInput=@n9), B=Transform(None=Constant3Vector() {constant=<rgba 0 0 -1 0>}) {transformSourceType=TRANSFORMSOURCE_View})
  n12 = ScalarParameter() {defaultValue=100, parameterName=Camfade_Offset}
  n13 = ScalarParameter() {defaultValue=800, parameterName=Camfade_Lemgth}
  n14 = Divide(A=Subtract(A=Multiply(A=@n10, B=@n11), B=@n12), B=@n13)
  n15 = Divide(A=Subtract(A=PixelDepth().R, B=@n12), B=@n13)
  n16 = StaticBool() {value=true}
  NamedRerouteDeclaration(None=Saturate(None=StaticSwitch(True=@n14, False=@n15, Value=@n16))) {name=CamFade, nodeColor=<rgba 1 0 0.9613 1>}
  n17 = Subtract(A=CameraPositionWS(), B=WorldPosition().XYZ)
  n18 = Abs(None=DotProduct(A=VertexNormalWS(), B=Normalize(VectorInput=@n17)))
  n19 = OneMinus(None=@n18)
  n20 = ScalarParameter() {parameterName="Fresnel_Invert (0or1)"}
  n21 = Subtract(
   A=Power(Base=LinearInterpolate(A=@n19, B=@n18, Alpha=@n20), Exp=ScalarParameter() {defaultValue=1.6128, parameterName=Fresnel_Power})
   B=ScalarParameter() {parameterName=Fresnel_Offset}
  )
  NamedRerouteDeclaration(None=VertexInterpolator(VS=Saturate(None=@n21)).PS) {name=Fresnel, nodeColor=<rgba 1 0 0.9613 1>}
  n22 = ScalarParameter() {defaultValue=599.6575, group=DepthFade, parameterName=DF_Distance}
  n23 = LinearInterpolate(A=@n22, B=@n22)
  n24 = Subtract(A=SceneDepth().R, B=PixelDepth().R)
  n25 = Subtract(A=Divide(A=Subtract(A=@n23, B=@n24), B=@n23), B=ScalarParameter() {group=DepthFade, parameterName=DF_Offset})
  n26 = ScalarParameter() {defaultValue=2, group=DepthFade, parameterName=DF_Power}
  NamedRerouteDeclaration(None=Saturate(None=Power(Base=@n25, Exp=@n26))) {name=DepthFadeValue, nodeColor=<rgba 0.5957 0 1 1>}
 ]
}
